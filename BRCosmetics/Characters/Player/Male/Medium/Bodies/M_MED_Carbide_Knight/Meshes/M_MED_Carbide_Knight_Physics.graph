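# Physics asset for the Carbide Knight medium male body, set up for secondary motion of the skirt.
# Front, side and back capsule chains on both sides hang from the pelvis with limited swing and zero twist; the front chains attach through small master bodies and swing on one axis at the top.
# Freely constrained thigh and calf capsules provide leg collision, and a huge root capsule below the character approximates a floor.
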PhysicsAsset /BRCosmetics/Characters/Player/Male/Medium/Bodies/M_MED_Carbide_Knight/Meshes/M_MED_Carbide_Knight_Physics
bodies {
 calf_l [{shapeType=Capsule, length=35.9055, radius=9.3735, shapeName=calf_l_capsule, center=<-21.8315 0 0>, rotation=<rot 90 0 0>}]
 calf_r [{shapeType=Capsule, length=36.2838, radius=10.1489, shapeName=calf_r_capsule, center=<21.8315 0 0>, rotation=<rot 90 0 0>}]
 pelvis [{shapeType=Capsule, length=1, radius=1, shapeName=pelvis_capsule}]
 root [{shapeType=Capsule, length=0, radius=600, shapeName=root_capsule, center=<0 0 -598.1915>, rotation=<rot 0 -90 0>}]
 skirtBack1_l [{shapeType=Capsule, length=6.25, radius=5, shapeName=skirtBack1_l_capsule, center=<5 3 0>, rotation=<rot 90 0 0>}]
 skirtBack1_r [{shapeType=Capsule, length=6.25, radius=5, shapeName=skirtBack1_r_capsule, center=<-5 -3 0>, rotation=<rot 90 0 0>}]
 skirtBack2_l [{shapeType=Capsule, length=6.25, radius=5, shapeName=skirtBack2_l_capsule, center=<5 3 0>, rotation=<rot 90 0 0>}]
 skirtBack2_r [{shapeType=Capsule, length=4.75, radius=6.5, shapeName=skirtBack2_r_capsule, center=<-5 -3 0>, rotation=<rot 90 0 0>}]
 skirtBack3_l [{shapeType=Capsule, length=6.25, radius=5, shapeName=skirtBack3_l_capsule, center=<5 3 0>, rotation=<rot 90 0 0>}]
 skirtBack3_r [{shapeType=Capsule, length=3.25, radius=8, shapeName=skirtBack3_r_capsule, center=<-5 -3 0>, rotation=<rot 90 0 0>}]
 skirtBack4_l [{shapeType=Capsule, length=6.25, radius=5, shapeName=skirtBack4_l_capsule, center=<5 3 0>, rotation=<rot 90 0 0>}]
 skirtBack4_r [{shapeType=Capsule, length=3.25, radius=8, shapeName=skirtBack4_r_capsule, center=<-5 -3 0>, rotation=<rot 90 0 0>}]
 skirtFront1_l [{shapeType=Capsule, length=6.5, radius=1.75, shapeName=skirtFront1_l_capsule, center=<7.6 0 0>, rotation=<rot 90 0 0>}]
 skirtFront1_r [{shapeType=Capsule, length=6.5, radius=1.75, shapeName=skirtFront1_r_capsule, center=<-7.6531 0 0>, rotation=<rot 90 0 0>}]
 skirtFront2_l [{shapeType=Capsule, length=9.5, radius=1.75, shapeName=skirtFront2_l_capsule, center=<5 0 0>, rotation=<rot 90 0 0>}]
 skirtFront2_r [{shapeType=Capsule, length=9.5, radius=1.75, shapeName=skirtFront2_r_capsule, center=<-5 0 0>, rotation=<rot 90 0 0>}]
 skirtFront3_l [{shapeType=Capsule, length=9.5, radius=1.75, shapeName=skirtFront3_l_capsule, center=<5 0 0>, rotation=<rot 90 0 0>}]
 skirtFront3_r [{shapeType=Capsule, length=8.5, radius=2.75, shapeName=skirtFront3_r_capsule, center=<-5 0 0>, rotation=<rot 90 0 0>}]
 skirtFront_MSTR_l [{shapeType=Capsule, length=0.505, radius=0.505, shapeName=skirtFront_MSTR_l_capsule, rotation=<rot 0 -90 0>}]
 skirtFront_MSTR_r [{shapeType=Capsule, length=0.505, radius=0.505, shapeName=skirtFront_MSTR_r_capsule, rotation=<rot 0 -90 0>}]
 skirtSide1_l [{shapeType=Capsule, length=6.25, radius=5, shapeName=skirtSide1_l_capsule, center=<5 3 0>, rotation=<rot 90 0 0>}]
 skirtSide1_r [{shapeType=Capsule, length=0.5, radius=9.5, shapeName=skirtSide1_r_capsule, center=<-8.1509 -8 0>, rotation=<rot 90 0 0>}]
 skirtSide2_l [{shapeType=Capsule, length=6.25, radius=5, shapeName=skirtSide2_l_capsule, center=<5 3 0>, rotation=<rot 90 0 0>}]
 skirtSide2_r [{shapeType=Capsule, length=5.5, radius=9.5, shapeName=skirtSide2_r_capsule, center=<-5 -8.5 0>, rotation=<rot 90 -19.0311 -19.031>}]
 skirtSide3_l [{shapeType=Capsule, length=6.25, radius=5, shapeName=skirtSide3_l_capsule, center=<5 3 0>, rotation=<rot 90 0 0>}]
 skirtSide3_r [{shapeType=Capsule, length=6, radius=9.75, shapeName=skirtSide3_r_capsule, center=<-5 -5 0>, rotation=<rot 90 0 0>}]
 skirtSide4_l [{shapeType=Capsule, length=6.25, radius=5, shapeName=skirtSide4_l_capsule, center=<5 3 0>, rotation=<rot 90 0 0>}]
 skirtSide4_r [{shapeType=Capsule, length=9.25, radius=9.5, shapeName=skirtSide4_r_capsule, center=<-5 -5 0>, rotation=<rot 90 0 0>}]
 thigh_l [{shapeType=Capsule, length=21.5, radius=8.75, shapeName=thigh_l_capsule, center=<-17 0 0>, rotation=<rot 90 0 0>}]
 thigh_r [{shapeType=Capsule, length=20.75, radius=8.75, shapeName=thigh_r_capsule, center=<15 0 0>, rotation=<rot 90 89.6476 84.6477>}]
}
constraints [
 {bone1Name=skirtSide4_l, bone2Name=skirtSide3_l, swing1LimitDegrees=45, swing2LimitDegrees=45, twistLimitDegrees=0}
 {bone1Name=skirtSide2_l, bone2Name=skirtSide1_l, swing1LimitDegrees=20, swing2LimitDegrees=20, twistLimitDegrees=0}
 {bone1Name=skirtSide3_l, bone2Name=skirtSide2_l, swing1LimitDegrees=45, swing2LimitDegrees=45, twistLimitDegrees=0}
 {bone1Name=skirtFront2_l, bone2Name=skirtFront1_l, swing1LimitDegrees=45, swing2LimitDegrees=45, twistLimitDegrees=0}
 {bone1Name=skirtFront3_l, bone2Name=skirtFront2_l, swing1LimitDegrees=45, swing2LimitDegrees=45, twistLimitDegrees=0}
 {bone1Name=skirtBack4_l, bone2Name=skirtBack3_l, swing1LimitDegrees=55, swing2LimitDegrees=55, twistLimitDegrees=0}
 {bone1Name=skirtFront3_r, bone2Name=skirtFront2_r, swing1LimitDegrees=45, swing2LimitDegrees=45, twistLimitDegrees=0}
 {bone1Name=skirtFront2_r, bone2Name=skirtFront1_r, swing1LimitDegrees=45, swing2LimitDegrees=45, twistLimitDegrees=0}
 {bone1Name=skirtBack2_l, bone2Name=skirtBack1_l, swing1LimitDegrees=25, swing2LimitDegrees=30, twistLimitDegrees=0}
 {bone1Name=skirtBack3_l, bone2Name=skirtBack2_l, swing1LimitDegrees=55, swing2LimitDegrees=55, twistLimitDegrees=0}
 {bone1Name=skirtBack4_r, bone2Name=skirtBack3_r, swing1LimitDegrees=45, swing2LimitDegrees=45, twistLimitDegrees=0}
 {bone1Name=skirtBack3_r, bone2Name=skirtBack2_r, swing1LimitDegrees=45, swing2LimitDegrees=45, twistLimitDegrees=0}
 {bone1Name=skirtBack2_r, bone2Name=skirtBack1_r, swing1LimitDegrees=20, swing2LimitDegrees=30, twistLimitDegrees=0}
 {bone1Name=skirtFront1_r, bone2Name=skirtFront_MSTR_r, swing1LimitDegrees=45, swing2LimitDegrees=0, twistLimitDegrees=0}
 {bone1Name=skirtFront1_l, bone2Name=skirtFront_MSTR_l, swing1LimitDegrees=45, swing2LimitDegrees=0, twistLimitDegrees=0}
 {bone1Name=skirtBack1_r, bone2Name=pelvis, swing1LimitDegrees=20, swing2LimitDegrees=30, twistLimitDegrees=0}
 {bone1Name=skirtBack1_l, bone2Name=pelvis, swing1LimitDegrees=25, swing2LimitDegrees=30, twistLimitDegrees=0}
 {bone1Name=skirtSide1_l, bone2Name=pelvis, swing1LimitDegrees=45, swing2LimitDegrees=45, twistLimitDegrees=0}
 {
  bone1Name=thigh_l
  bone2Name=pelvis
  swing1LimitDegrees=45
  swing1Motion=Free
  swing2LimitDegrees=45
  swing2Motion=Free
  twistLimitDegrees=45
  twistMotion=Free
 }
 {
  bone1Name=thigh_r
  bone2Name=pelvis
  swing1LimitDegrees=45
  swing1Motion=Free
  swing2LimitDegrees=45
  swing2Motion=Free
  twistLimitDegrees=45
  twistMotion=Free
 }
 {
  bone1Name=calf_l
  bone2Name=thigh_l
  swing1LimitDegrees=45
  swing1Motion=Free
  swing2LimitDegrees=45
  swing2Motion=Free
  twistLimitDegrees=45
  twistMotion=Free
 }
 {
  bone1Name=calf_r
  bone2Name=thigh_r
  swing1LimitDegrees=45
  swing1Motion=Free
  swing2LimitDegrees=45
  swing2Motion=Free
  twistLimitDegrees=45
  twistMotion=Free
 }
 {bone1Name=skirtSide1_r, bone2Name=pelvis, swing1LimitDegrees=15, swing2LimitDegrees=45, twistLimitDegrees=0}
 {bone1Name=skirtSide2_r, bone2Name=skirtSide1_r, swing1LimitDegrees=15, swing2LimitDegrees=45, twistLimitDegrees=0}
 {bone1Name=skirtSide3_r, bone2Name=skirtSide2_r, swing1LimitDegrees=45, swing2LimitDegrees=45, twistLimitDegrees=0}
 {bone1Name=skirtSide4_r, bone2Name=skirtSide3_r, swing1LimitDegrees=45, swing2LimitDegrees=45, twistLimitDegrees=0}
 {
  bone1Name=pelvis
  bone2Name=root
  swing1LimitDegrees=45
  swing1Motion=Free
  swing2LimitDegrees=45
  swing2Motion=Free
  twistLimitDegrees=45
  twistMotion=Free
 }
 {
  bone1Name=skirtFront_MSTR_l
  bone2Name=pelvis
  swing1LimitDegrees=45
  swing1Motion=Free
  swing2LimitDegrees=45
  swing2Motion=Free
  twistLimitDegrees=45
  twistMotion=Free
 }
 {
  bone1Name=skirtFront_MSTR_r
  bone2Name=pelvis
  swing1LimitDegrees=45
  swing1Motion=Free
  swing2LimitDegrees=45
  swing2Motion=Free
  twistLimitDegrees=45
  twistMotion=Free
 }
]
