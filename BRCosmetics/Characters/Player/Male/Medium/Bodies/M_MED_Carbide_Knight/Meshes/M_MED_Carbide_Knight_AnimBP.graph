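# Cosmetic Anim Blueprint for the male Carbide Knight body, correcting shoulder pads, hip pads and skirt armor on the character.
# Copies the parent mesh pose through a deform Control Rig, then chains additive bone modifiers around pelvis-space RigidBody simulation.
# Each update derives modifier alphas from limb and head deltas plus frontend, skydive and targeting states; parachute events set glider alphas.
# Certain vehicle seats disable the physics and shrink back and side skirt bones.
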AnimBlueprint /BRCosmetics/Characters/Player/Male/Medium/Bodies/M_MED_Carbide_Knight/Meshes/M_MED_Carbide_Knight_AnimBP
target_skeleton=&/M_MED_Carbide_Knight_Skeleton
graphs [
 AnimGraph {nodes [
   n1 = Control Rig(Source=Copy Pose From Mesh() {bUseAttachedParent=true}) {
    alphaBoolBlend {blendOption=Linear}
    controlRigAssetReference {blueprintRigClass=&/M_MED_Carbide_DeformRig_Master_ControlRig1.M_MED_Carbide_DeformRig_Master_ControlRig1_C}
   }
   n2 = "Transform (Modify) Bone - Bone: skirtFront_MSTR_r"(
    ComponentPose=Local To Component(LocalPose=@n1)
    Alpha=Get thigh_up_r() -> (thigh_up_r:double)
    Translation="4.024,-0.000,-3.005"
    Rotation="0,0,0"
    Scale="1,1,1"
   ) {
    rotationMode=BMM_Additive
    rotationSpace=BCS_BoneSpace
    translationMode=BMM_Additive
    alphaBoolBlend {blendOption=Linear}
    boneToModify {boneName=skirtFront_MSTR_r}
    translation {x=4.024, z=-3.005}
   }
   n3 = "Transform (Modify) Bone - Bone: skirtFront_MSTR_l"(ComponentPose=@n2, Alpha=Get thigh_up_l() -> (thigh_up_l:double), Translation="-4.024,-0.000,-3.005", Rotation="0,0,0", Scale="1,1,1") {
    rotationMode=BMM_Additive
    rotationSpace=BCS_BoneSpace
    translationMode=BMM_Additive
    alphaBoolBlend {blendOption=Linear}
    boneToModify {boneName=skirtFront_MSTR_l}
    translation {x=-4.024, z=-3.005}
   }
   n4 = Get frontendNoEmote() -> (frontendNoEmote:double)
   n5 = "Transform (Modify) Bone - Bone: skirtSide1_l"(ComponentPose=@n3, Alpha=@n4, Translation="4,0,0", Rotation="0,0,0", Scale="1,1,1") {
    rotationMode=BMM_Additive
    rotationSpace=BCS_BoneSpace
    translationMode=BMM_Additive
    alphaBoolBlend {blendOption=Linear}
    boneToModify {boneName=skirtSide1_l}
    translation {x=4}
   }
   n6 = "Transform (Modify) Bone - Bone: Hip_Pad_l"(ComponentPose=@n5, Alpha=@n4, Translation="0,0,0", Rotation="3,0,0", Scale="1,1,1") {
    rotationMode=BMM_Additive
    rotationSpace=BCS_BoneSpace
    translationMode=BMM_Additive
    alphaBoolBlend {blendOption=Linear}
    boneToModify {boneName=Hip_Pad_l}
    rotation {pitch=3}
   }
   n7 = RigidBody!(
    ComponentPose=@n6
    Alpha=Get in vehicle rigid alpha() -> (in vehicle rigid alpha:double)
    OverrideWorldGravity:Vector=Get GravityOverride()
    ExternalForce
   ) -> (Pose) {
    bClampLinearTranslationLimitToRefPose=true
    bOverrideWorldGravity=true
    overridePhysicsAsset=&/M_MED_Carbide_Knight_Physics
    simulationSpace=BaseBoneSpace
    alphaBoolBlend {blendOption=Linear}
    baseBoneRef {boneName=pelvis}
    componentAppliedLinearAccClamp=<3000 3000 3000>
    componentLinearAccScale=<0.5 0.3 0.3>
    componentLinearVelScale=<0.5 0.3 0.3>
    overrideWorldGravity {z=-980}
    simSpaceSettings {
     maxAngularAcceleration=1000
     maxAngularVelocity=3
     maxLinearAcceleration=700
     maxLinearVelocity=500
     velocityScaleZ=0
     worldAlpha=1
     externalLinearDragV=<0.2 0.2>
    }
   }
   n8 = "Transform (Modify) Bone - Bone: upperarmShoulderPadMain_l"(ComponentPose=@n7, Alpha=Get clavicleLRot() -> (clavicleLRot:double), Translation="0,0,0", Rotation="0,-20,0", Scale="1,1,1") {
    rotationMode=BMM_Additive
    rotationSpace=BCS_BoneSpace
    translationMode=BMM_Additive
    alphaBoolBlend {blendOption=Linear}
    boneToModify {boneName=upperarmShoulderPadMain_l}
    rotation {yaw=-20}
   }
   n9 = "Transform (Modify) Bone - Bone: upperarmShoulderPadMain_r"(ComponentPose=@n8, Alpha=Get clavicleRRot() -> (clavicleRRot:double), Translation="2,0,0", Rotation="0,0,0", Scale="1,1,1") {
    rotationMode=BMM_Additive
    rotationSpace=BCS_BoneSpace
    translationMode=BMM_Additive
    alphaBoolBlend {blendOption=Linear}
    boneToModify {boneName=upperarmShoulderPadMain_r}
    translation {x=2}
   }
   n10 = "Transform (Modify) Bone - Bone: upperarmShoulderPadMain_l"(ComponentPose=@n9, Alpha=Get upperarm_rotZ_l_Up() -> (upperarm_rotZ_l_Up:double), Translation="0,0,-1", Rotation="-20,0,0", Scale="1,1,1") {
    rotationMode=BMM_Additive
    rotationSpace=BCS_BoneSpace
    translationMode=BMM_Additive
    alphaBoolBlend {blendOption=Linear}
    boneToModify {boneName=upperarmShoulderPadMain_l}
    rotation {pitch=-20}
    translation {z=-1}
   }
   n11 = "Transform (Modify) Bone - Bone: upperarmShoulderPadMain_r"(ComponentPose=@n10, Alpha=Get upperarm_rotZ_r_Up() -> (upperarm_rotZ_r_Up:double), Translation="0,0,-1", Rotation="-20,0,0", Scale="1,1,1") {
    rotationMode=BMM_Additive
    rotationSpace=BCS_BoneSpace
    alphaBoolBlend {blendOption=Linear}
    boneToModify {boneName=upperarmShoulderPadMain_r}
    rotation {pitch=-20}
    translation {z=-1}
   }
   n12 = "Transform (Modify) Bone - Bone: upperarmShoulderPadMain_r"(ComponentPose=@n11, Alpha=Get upperarm_rotY_r() -> (upperarm_rotY_r:double), Translation="0,0,0", Rotation="0,-10,20", Scale="1,0.8,1") {
    rotationMode=BMM_Additive
    rotationSpace=BCS_BoneSpace
    alphaBoolBlend {blendOption=Linear}
    boneToModify {boneName=upperarmShoulderPadMain_r}
    rotation {roll=20, yaw=-10}
    scale {y=0.8}
   }
   n13 = "Transform (Modify) Bone - Bone: upperarmShoulderPadMain_l"(ComponentPose=@n12, Alpha=Get upperarm_rotY_l() -> (upperarm_rotY_l:double), Translation="0,0,0", Rotation="0,0,0", Scale="1,1,1") {
    rotationMode=BMM_Additive
    rotationSpace=BCS_BoneSpace
    alphaBoolBlend {blendOption=Linear}
    boneToModify {boneName=upperarmShoulderPadMain_l}
   }
   n14 = "Transform (Modify) Bone - Bone: Hip_Pad_l"(ComponentPose=@n13, Alpha=Get thigh_up_l(), Translation="1.371,-2.014,0.507", Rotation="0,-20,-25", Scale="1,1,1") {
    rotationMode=BMM_Additive
    rotationSpace=BCS_BoneSpace
    translationMode=BMM_Additive
    alphaBoolBlend {blendOption=Linear}
    alphaScaleBiasClamp {bInterpResult=true, interpSpeedDecreasing=15, interpSpeedIncreasing=15}
    boneToModify {boneName=Hip_Pad_l}
    rotation {roll=-25, yaw=-20}
    translation=<1.371 -2.014 0.507>
   }
   n15 = "Transform (Modify) Bone - Bone: Hip_Pad_l"(ComponentPose=@n14, Alpha=Get ThighSideLf() -> (ThighSideLf:double), Translation="0,0,0", Rotation="10,0,0", Scale="1,1,1") {
    rotationMode=BMM_Additive
    rotationSpace=BCS_BoneSpace
    translationMode=BMM_Additive
    alphaBoolBlend {blendOption=Linear}
    alphaScaleBiasClamp {bInterpResult=true, interpSpeedDecreasing=15, interpSpeedIncreasing=15}
    boneToModify {boneName=Hip_Pad_l}
    rotation {pitch=10}
   }
   n16 = "Transform (Modify) Bone - Bone: Hip_Pad_r"(ComponentPose=@n15, Alpha=Get thigh_up_r(), Translation="1.371,-2.014,0.507", Rotation="0,-20,-25", Scale="1,1,1") {
    rotationMode=BMM_Additive
    rotationSpace=BCS_BoneSpace
    translationMode=BMM_Additive
    alphaBoolBlend {blendOption=Linear}
    alphaScaleBiasClamp {bInterpResult=true, interpSpeedDecreasing=15, interpSpeedIncreasing=15}
    boneToModify {boneName=Hip_Pad_r}
    rotation {roll=-25, yaw=-20}
    translation=<1.371 -2.014 0.507>
   }
   n17 = "Transform (Modify) Bone - Bone: Hip_Pad_r"(ComponentPose=@n16, Alpha=Get ThighSideRt() -> (ThighSideRt:double), Translation="0,0,0", Rotation="10,0,0", Scale="1,1,1") {
    rotationMode=BMM_Additive
    rotationSpace=BCS_BoneSpace
    translationMode=BMM_Additive
    alphaBoolBlend {blendOption=Linear}
    alphaScaleBiasClamp {bInterpResult=true, interpSpeedDecreasing=15, interpSpeedIncreasing=15}
    boneToModify {boneName=Hip_Pad_r}
    rotation {pitch=10}
   }
   n18 = "Transform (Modify) Bone - Bone: Hip_Pad_r"(ComponentPose=@n17, Alpha=1, Translation="0,0,0", Rotation="5,0,0", Scale="1,1,1") {
    rotationMode=BMM_Additive
    rotationSpace=BCS_BoneSpace
    translationMode=BMM_Additive
    alphaBoolBlend {blendOption=Linear}
    alphaScaleBiasClamp {bInterpResult=true, interpSpeedDecreasing=15, interpSpeedIncreasing=15}
    boneToModify {boneName=Hip_Pad_r}
    rotation {pitch=5}
   }
   n19 = Get Is Sky Diving() -> (Is Sky Diving:double)
   n20 = "Transform (Modify) Bone - Bone: skirtSide1_l"(ComponentPose=@n18, Alpha=@n19, Translation="0,0,0", Rotation="0,10,0", Scale="1,1,1") {
    rotationMode=BMM_Additive
    rotationSpace=BCS_BoneSpace
    translationMode=BMM_Additive
    alphaBoolBlend {blendOption=Linear}
    alphaScaleBiasClamp {bInterpResult=true, interpSpeedDecreasing=15, interpSpeedIncreasing=15}
    boneToModify {boneName=skirtSide1_l}
    rotation {yaw=10}
   }
   n21 = "Transform (Modify) Bone - Bone: Hip_Pad_l"(ComponentPose=@n20, Alpha=@n19, Translation="0,0,0", Rotation="10,0,0", Scale="1,1,1") {
    rotationMode=BMM_Additive
    rotationSpace=BCS_BoneSpace
    translationMode=BMM_Additive
    alphaBoolBlend {blendOption=Linear}
    alphaScaleBiasClamp {bInterpResult=true, interpSpeedDecreasing=15, interpSpeedIncreasing=15}
    boneToModify {boneName=Hip_Pad_l}
    rotation {pitch=10}
   }
   n22 = "Transform (Modify) Bone - Bone: Hip_Pad_r"(ComponentPose=@n21, Alpha=@n19, Translation="0,0,0", Rotation="10,0,0", Scale="1,1,1") {
    rotationMode=BMM_Additive
    rotationSpace=BCS_BoneSpace
    translationMode=BMM_Additive
    alphaBoolBlend {blendOption=Linear}
    alphaScaleBiasClamp {bInterpResult=true, interpSpeedDecreasing=15, interpSpeedIncreasing=15}
    boneToModify {boneName=Hip_Pad_r}
    rotation {pitch=10}
   }
   n23 = "Transform (Modify) Bone - Bone: upperarmShoulderPadMain_l"(ComponentPose=@n22, Alpha=@n19, Translation="4,0,2", Rotation="0,0,0", Scale="0.8,0.9,0.9") {
    rotationMode=BMM_Additive
    rotationSpace=BCS_BoneSpace
    scaleMode=BMM_Additive
    translationMode=BMM_Additive
    alphaBoolBlend {blendOption=Linear}
    boneToModify {boneName=upperarmShoulderPadMain_l}
    scale=<0.8 0.9 0.9>
    translation {x=4, z=2}
   }
   n24 = "Transform (Modify) Bone - Bone: upperarmShoulderPadMain_r"(ComponentPose=@n23, Alpha=@n19, Translation="-4,0,2", Rotation="0,0,0", Scale="0.8,0.9,0.9") {
    rotationMode=BMM_Additive
    rotationSpace=BCS_BoneSpace
    scaleMode=BMM_Additive
    translationMode=BMM_Additive
    alphaBoolBlend {blendOption=Linear}
    boneToModify {boneName=upperarmShoulderPadMain_r}
    scale=<0.8 0.9 0.9>
    translation {x=-4, z=2}
   }
   n25 = Get Is Parachute Opened() -> (Is Parachute Opened:double)
   n26 = "Transform (Modify) Bone - Bone: upperarmShoulderPadMain_r"(ComponentPose=@n24, Alpha=@n25, Translation="-2,0,0", Rotation="0,0,0", Scale="1,1,1") {
    rotationMode=BMM_Additive
    rotationSpace=BCS_BoneSpace
    scaleMode=BMM_Additive
    translationMode=BMM_Additive
    alphaBoolBlend {blendOption=Linear}
    boneToModify {boneName=upperarmShoulderPadMain_r}
    translation {x=-2}
   }
   n27 = "Transform (Modify) Bone - Bone: upperarmShoulderPadMain_l"(ComponentPose=@n26, Alpha=@n25, Translation="2,0,0", Rotation="0,0,0", Scale="1,1,1") {
    rotationMode=BMM_Additive
    rotationSpace=BCS_BoneSpace
    scaleMode=BMM_Additive
    translationMode=BMM_Additive
    alphaBoolBlend {blendOption=Linear}
    boneToModify {boneName=upperarmShoulderPadMain_l}
    translation {x=2}
   }
   n28 = "Transform (Modify) Bone - Bone: spine_05_headBackCollision"(ComponentPose=@n27, Alpha=Get headRotBack() -> (headRotBack:double), Translation="0,0,0", Rotation="0,0,0", Scale="0.3,1,1") {
    rotationMode=BMM_Additive
    rotationSpace=BCS_BoneSpace
    scaleMode=BMM_Additive
    translationMode=BMM_Additive
    alphaBoolBlend {blendOption=Linear}
    boneToModify {boneName=spine_05_headBackCollision}
    scale {x=0.3}
   }
   n29 = "Transform (Modify) Bone - Bone: upperarmShoulderPadMain_r"(ComponentPose=@n28, Alpha=Get Is Targeting() -> (Is Targeting:double), Translation="0,0,0", Rotation="0,0,0", Scale="0.9,0.85,0.9") {
    rotationMode=BMM_Additive
    rotationSpace=BCS_BoneSpace
    scaleMode=BMM_Additive
    translationMode=BMM_Additive
    alphaBoolBlend {blendOption=Linear}
    alphaScaleBiasClamp {bInterpResult=true, interpSpeedDecreasing=50, interpSpeedIncreasing=50}
    boneToModify {boneName=upperarmShoulderPadMain_r}
    scale=<0.9 0.85 0.9>
   }
   n30 = Get in vehicle scale alpha() -> (in vehicle scale alpha:double)
   n31 = "Transform (Modify) Bone - Bone: skirtBack1_r"(ComponentPose=@n29, Alpha=@n30, Translation="0,0,0", Rotation="0,0,0", Scale="0.01,0.01,0.01") {
    rotationMode=BMM_Additive
    rotationSpace=BCS_BoneSpace
    scaleMode=BMM_Additive
    translationMode=BMM_Additive
    alphaBoolBlend {blendOption=Linear}
    alphaScaleBiasClamp {bInterpResult=true, interpSpeedDecreasing=50, interpSpeedIncreasing=50}
    boneToModify {boneName=skirtBack1_r}
    scale=<0.01 0.01 0.01>
   }
   n32 = "Transform (Modify) Bone - Bone: skirtBack2_r"(ComponentPose=@n31, Alpha=@n30, Translation="0,0,0", Rotation="0,0,0", Scale="0.01,0.01,0.01") {
    rotationMode=BMM_Additive
    rotationSpace=BCS_BoneSpace
    scaleMode=BMM_Additive
    translationMode=BMM_Additive
    alphaBoolBlend {blendOption=Linear}
    alphaScaleBiasClamp {bInterpResult=true, interpSpeedDecreasing=50, interpSpeedIncreasing=50}
    boneToModify {boneName=skirtBack2_r}
    scale=<0.01 0.01 0.01>
   }
   n33 = "Transform (Modify) Bone - Bone: skirtBack3_r"(ComponentPose=@n32, Alpha=@n30, Translation="0,0,0", Rotation="0,0,0", Scale="0.01,0.01,0.01") {
    rotationMode=BMM_Additive
    rotationSpace=BCS_BoneSpace
    scaleMode=BMM_Additive
    translationMode=BMM_Additive
    alphaBoolBlend {blendOption=Linear}
    alphaScaleBiasClamp {bInterpResult=true, interpSpeedDecreasing=50, interpSpeedIncreasing=50}
    boneToModify {boneName=skirtBack3_r}
    scale=<0.01 0.01 0.01>
   }
   n34 = "Transform (Modify) Bone - Bone: skirtBack4_r"(ComponentPose=@n33, Alpha=@n30, Translation="0,0,0", Rotation="0,0,0", Scale="0.01,0.01,0.01") {
    rotationMode=BMM_Additive
    rotationSpace=BCS_BoneSpace
    scaleMode=BMM_Additive
    translationMode=BMM_Additive
    alphaBoolBlend {blendOption=Linear}
    alphaScaleBiasClamp {bInterpResult=true, interpSpeedDecreasing=50, interpSpeedIncreasing=50}
    boneToModify {boneName=skirtBack4_r}
    scale=<0.01 0.01 0.01>
   }
   n35 = "Transform (Modify) Bone - Bone: skirtBack1_l"(ComponentPose=@n34, Alpha=@n30, Translation="0,0,0", Rotation="0,0,0", Scale="0.01,0.01,0.01") {
    rotationMode=BMM_Additive
    rotationSpace=BCS_BoneSpace
    scaleMode=BMM_Additive
    translationMode=BMM_Additive
    alphaBoolBlend {blendOption=Linear}
    alphaScaleBiasClamp {bInterpResult=true, interpSpeedDecreasing=50, interpSpeedIncreasing=50}
    boneToModify {boneName=skirtBack1_l}
    scale=<0.01 0.01 0.01>
   }
   n36 = "Transform (Modify) Bone - Bone: skirtBack2_l"(ComponentPose=@n35, Alpha=@n30, Translation="0,0,0", Rotation="0,0,0", Scale="0.01,0.01,0.01") {
    rotationMode=BMM_Additive
    rotationSpace=BCS_BoneSpace
    scaleMode=BMM_Additive
    translationMode=BMM_Additive
    alphaBoolBlend {blendOption=Linear}
    alphaScaleBiasClamp {bInterpResult=true, interpSpeedDecreasing=50, interpSpeedIncreasing=50}
    boneToModify {boneName=skirtBack2_l}
    scale=<0.01 0.01 0.01>
   }
   n37 = "Transform (Modify) Bone - Bone: skirtBack3_l"(ComponentPose=@n36, Alpha=@n30, Translation="0,0,0", Rotation="0,0,0", Scale="0.01,0.01,0.01") {
    rotationMode=BMM_Additive
    rotationSpace=BCS_BoneSpace
    scaleMode=BMM_Additive
    translationMode=BMM_Additive
    alphaBoolBlend {blendOption=Linear}
    alphaScaleBiasClamp {bInterpResult=true, interpSpeedDecreasing=50, interpSpeedIncreasing=50}
    boneToModify {boneName=skirtBack3_l}
    scale=<0.01 0.01 0.01>
   }
   n38 = "Transform (Modify) Bone - Bone: skirtBack4_l"(ComponentPose=@n37, Alpha=@n30, Translation="0,0,0", Rotation="0,0,0", Scale="0.01,0.01,0.01") {
    rotationMode=BMM_Additive
    rotationSpace=BCS_BoneSpace
    scaleMode=BMM_Additive
    translationMode=BMM_Additive
    alphaBoolBlend {blendOption=Linear}
    alphaScaleBiasClamp {bInterpResult=true, interpSpeedDecreasing=50, interpSpeedIncreasing=50}
    boneToModify {boneName=skirtBack4_l}
    scale=<0.01 0.01 0.01>
   }
   n39 = "Transform (Modify) Bone - Bone: skirtSide1_r"(ComponentPose=@n38, Alpha=@n30, Translation="0,0,0", Rotation="0,0,0", Scale="0.01,0.01,0.01") {
    rotationMode=BMM_Additive
    rotationSpace=BCS_BoneSpace
    scaleMode=BMM_Additive
    translationMode=BMM_Additive
    alphaBoolBlend {blendOption=Linear}
    alphaScaleBiasClamp {bInterpResult=true, interpSpeedDecreasing=50, interpSpeedIncreasing=50}
    boneToModify {boneName=skirtSide1_r}
    scale=<0.01 0.01 0.01>
   }
   n40 = "Transform (Modify) Bone - Bone: skirtSide2_r"(ComponentPose=@n39, Alpha=@n30, Translation="0,0,0", Rotation="0,0,0", Scale="0.01,0.01,0.01") {
    rotationMode=BMM_Additive
    rotationSpace=BCS_BoneSpace
    scaleMode=BMM_Additive
    translationMode=BMM_Additive
    alphaBoolBlend {blendOption=Linear}
    alphaScaleBiasClamp {bInterpResult=true, interpSpeedDecreasing=50, interpSpeedIncreasing=50}
    boneToModify {boneName=skirtSide2_r}
    scale=<0.01 0.01 0.01>
   }
   n41 = "Transform (Modify) Bone - Bone: skirtSide3_r"(ComponentPose=@n40, Alpha=@n30, Translation="0,0,0", Rotation="0,0,0", Scale="0.01,0.01,0.01") {
    rotationMode=BMM_Additive
    rotationSpace=BCS_BoneSpace
    scaleMode=BMM_Additive
    translationMode=BMM_Additive
    alphaBoolBlend {blendOption=Linear}
    alphaScaleBiasClamp {bInterpResult=true, interpSpeedDecreasing=50, interpSpeedIncreasing=50}
    boneToModify {boneName=skirtSide3_r}
    scale=<0.01 0.01 0.01>
   }
   n42 = "Transform (Modify) Bone - Bone: skirtSide4_r"(ComponentPose=@n41, Alpha=@n30, Translation="0,0,0", Rotation="0,0,0", Scale="0.01,0.01,0.01") {
    rotationMode=BMM_Additive
    rotationSpace=BCS_BoneSpace
    scaleMode=BMM_Additive
    translationMode=BMM_Additive
    alphaBoolBlend {blendOption=Linear}
    alphaScaleBiasClamp {bInterpResult=true, interpSpeedDecreasing=50, interpSpeedIncreasing=50}
    boneToModify {boneName=skirtSide4_r}
    scale=<0.01 0.01 0.01>
   }
   n43 = "Transform (Modify) Bone - Bone: skirtSide1_l"(ComponentPose=@n42, Alpha=@n30, Translation="0,0,0", Rotation="0,0,0", Scale="0.01,0.01,0.01") {
    rotationMode=BMM_Additive
    rotationSpace=BCS_BoneSpace
    scaleMode=BMM_Additive
    translationMode=BMM_Additive
    alphaBoolBlend {blendOption=Linear}
    alphaScaleBiasClamp {bInterpResult=true, interpSpeedDecreasing=50, interpSpeedIncreasing=50}
    boneToModify {boneName=skirtSide1_l}
    scale=<0.01 0.01 0.01>
   }
   n44 = "Transform (Modify) Bone - Bone: skirtSide2_l"(ComponentPose=@n43, Alpha=@n30, Translation="0,0,0", Rotation="0,0,0", Scale="0.01,0.01,0.01") {
    rotationMode=BMM_Additive
    rotationSpace=BCS_BoneSpace
    scaleMode=BMM_Additive
    translationMode=BMM_Additive
    alphaBoolBlend {blendOption=Linear}
    alphaScaleBiasClamp {bInterpResult=true, interpSpeedDecreasing=50, interpSpeedIncreasing=50}
    boneToModify {boneName=skirtSide2_l}
    scale=<0.01 0.01 0.01>
   }
   n45 = "Transform (Modify) Bone - Bone: skirtSide3_l"(ComponentPose=@n44, Alpha=@n30, Translation="0,0,0", Rotation="0,0,0", Scale="0.01,0.01,0.01") {
    rotationMode=BMM_Additive
    rotationSpace=BCS_BoneSpace
    scaleMode=BMM_Additive
    translationMode=BMM_Additive
    alphaBoolBlend {blendOption=Linear}
    alphaScaleBiasClamp {bInterpResult=true, interpSpeedDecreasing=50, interpSpeedIncreasing=50}
    boneToModify {boneName=skirtSide3_l}
    scale=<0.01 0.01 0.01>
   }
   n46 = "Transform (Modify) Bone - Bone: skirtSide4_l"(ComponentPose=@n45, Alpha=@n30, Translation="0,0,0", Rotation="0,0,0", Scale="0.01,0.01,0.01") {
    rotationMode=BMM_Additive
    rotationSpace=BCS_BoneSpace
    scaleMode=BMM_Additive
    translationMode=BMM_Additive
    alphaBoolBlend {blendOption=Linear}
    alphaScaleBiasClamp {bInterpResult=true, interpSpeedDecreasing=50, interpSpeedIncreasing=50}
    boneToModify {boneName=skirtSide4_l}
    scale=<0.01 0.01 0.01>
   }
   Output Pose(Result=Component To Local(ComponentPose=@n46)) {layerGroup=None}
  ]}
 EventGraph {nodes [
   Event BlueprintUpdateAnimation()
   > n1 = Gravity Override State Function DEPRECATED(CharacterPartAnimInstance=Self-Reference(), GameTimeInSeconds=GetGameTimeInSeconds())
   > Set GravityOverride(GravityOverride=@n1)
   n2 = GetPartSkeletalMeshComponent()
   > Set upperarm_rotZ_r_Up(upperarm_rotZ_r_Up:double=GetAlphaFromReferenceTransformDelta(SkeletalMeshComponent=@n2, BoneName=upperarm_r, BaseName=clavicle_r, DeltaComponentType=EulerY, InputRange=(X=-20,Y=20), TwistAxis="0,0,1")) -> (Output_Get:double)
   > Set upperarm_rotZ_l_Up(upperarm_rotZ_l_Up:double=GetAlphaFromReferenceTransformDelta(SkeletalMeshComponent=@n2, BoneName=upperarm_l, BaseName=clavicle_l, DeltaComponentType=EulerY, InputRange=(X=-20,Y=20), TwistAxis="0,0,1")) -> (Output_Get:double)
   n3 = GetPartSkeletalMeshComponent()
   > Set upperarm_rotY_r(upperarm_rotY_r:double=GetAlphaFromReferenceTransformDelta(SkeletalMeshComponent=@n3, BoneName=upperarm_r, BaseName=clavicle_r, DeltaComponentType=EulerY, InputRange=(X=-40,Y=0), TwistAxis="0,1,0")) -> (Output_Get:double)
   > Set upperarm_rotY_l(upperarm_rotY_l:double=GetAlphaFromReferenceTransformDelta(SkeletalMeshComponent=@n3, BoneName=upperarm_l, BaseName=clavicle_l, DeltaComponentType=EulerY, InputRange=(X=0,Y=40), TwistAxis="0,1,0")) -> (Output_Get:double)
   n4 = BreakTransform(InTransform=GetDeltaTransformFromRefPose(self=GetOwnerSkeletalMeshComponent(), BoneName=thigh_l, BaseName=pelvis))
   > Set thigh_up_l(thigh_up_l:double=MapRangeClamped(Value=BreakRotator(InRot=@n4.Rotation).Yaw, InRangeA=0, InRangeB=-40)) -> (Output_Get:double)
   n5 = BreakTransform(InTransform=GetDeltaTransformFromRefPose(self=GetOwnerSkeletalMeshComponent(), BoneName=thigh_r, BaseName=pelvis))
   > Set thigh_up_r(thigh_up_r:double=MapRangeClamped(Value=BreakRotator(InRot=@n5.Rotation).Yaw, InRangeA=0, InRangeB=-40)) -> (Output_Get:double)
   n6 = BreakTransform(InTransform=GetDeltaTransformFromRefPose(self=GetOwnerSkeletalMeshComponent(), BoneName=thigh_l, BaseName=pelvis))
   > Set ThighSideLf(ThighSideLf:double=MapRangeClamped(Value=BreakRotator(InRot=@n6.Rotation).Pitch, InRangeA=0, InRangeB=40)) -> (Output_Get:double)
   n7 = BreakTransform(InTransform=GetDeltaTransformFromRefPose(self=GetOwnerSkeletalMeshComponent(), BoneName=thigh_r, BaseName=pelvis))
   > Set ThighSideRt(ThighSideRt:double=MapRangeClamped(Value=BreakRotator(InRot=@n7.Rotation).Pitch, InRangeA=0, InRangeB=40)) -> (Output_Get:double)
   n8 = BreakTransform(InTransform=GetDeltaTransformFromRefPose(self=GetOwnerSkeletalMeshComponent(), BoneName=thigh_l, BaseName=pelvis))
   > Set ThighSideLfIn(ThighSideLfIn:double=MapRangeClamped(Value=BreakRotator(InRot=@n8.Rotation).Pitch, InRangeA=0, InRangeB=-30)) -> (Output_Get:double)
   n9 = Get bIsSkydiveDiveMode()
   > Set Is Sky Diving(Is Sky Diving:double="To Float (Boolean)"(InBool=NOT Boolean(A=@n9))) -> (Output_Get:double)
   > Set Is Sky Diving(Is Sky Diving="To Float (Boolean)"(InBool=Get bIsSkydiveFloating()))
   n10 = Truncate(A=Get Is Parachute Opened())
   > n11 = Branch(Condition=AND Boolean(A=Get bIsSkyDiving(), B="To Boolean (Integer)"!(InInt:int=@n10) -> (ReturnValue:bool)))
   > n12 = Set Is Sky Diving(Is Sky Diving=1)
   Set headRotBack(
    execute=[@n11.else, @n12.then]
    headRotBack:double=GetAlphaFromReferenceTransformDelta(SkeletalMeshComponent=GetOwnerSkeletalMeshComponent(), BoneName=head, BaseName=spine_05, InputRange=(X=0,Y=-50), TwistAxis="1,0,0")
   ) -> (Output_Get:double)
   > Set Is Targeting(Is Targeting:double="To Float (Boolean)"(InBool=Get bIsTargeting())) -> (Output_Get:double)
   n13 = GetOwnerSkeletalMeshComponent()
   > Set clavicleLRot(clavicleLRot:double=GetAlphaFromReferenceTransformDelta(
     SkeletalMeshComponent=@n13
     BoneName=clavicle_l
     BaseName=spine_05
     DeltaComponentType=EulerY
     InputRange=(X=40,Y=-20)
     OutputRange=(X=1,Y=0)
     TwistAxis="1,0,0"
    )) -> (Output_Get:double)
   > Set clavicleRRot(clavicleRRot:double=GetAlphaFromReferenceTransformDelta(SkeletalMeshComponent=@n13, BoneName=clavicle_r, BaseName=spine_05, InputRange=(X=40,Y=-20), OutputRange=(X=1,Y=0), TwistAxis="1,0,0")) -> (Output_Get:double)
   n14 = NOT Boolean(A=Get bIsPlayingEmote())
   n15 = Get bIsInFrontEnd()
   > Set frontendNoEmote(frontendNoEmote:double="To Float (Boolean)"(InBool=AND Boolean(A=@n15, B=@n14))) -> (Output_Get:double)
   > n16 = Branch(Condition=OR Boolean!(
     A=Get bIsOstrichVehicleDriver() -> (bIsOstrichVehicleDriver:bool)
     B=Get bIsOstrichVehicleGunner() -> (bIsOstrichVehicleGunner:bool)
     C:bool=Get bIsFerretVehicleDriver() -> (bIsFerretVehicleDriver:bool)
     D:bool=Get bIsValetDriver() -> (bIsValetDriver:bool)
     E:bool=Get bIsGolfCartVehicleDriver() -> (bIsGolfCartVehicleDriver:bool)
     F:bool=Get bIsGolfCartVehicleFrontPassenger() -> (bIsGolfCartVehicleFrontPassenger:bool)
     G:bool=Get bIsOctopusVehicleDriver() -> (bIsOctopusVehicleDriver:bool)
     H:bool=Get bIsAntelopeVehicleDriver() -> (bIsAntelopeVehicleDriver:bool)
     I:bool=Get bIsMeatballDriver() -> (bIsMeatballDriver:bool)
    ) -> (ReturnValue))
   > Set in vehicle rigid alpha(in vehicle rigid alpha:double=0) -> (Output_Get:double)
   > Set in vehicle scale alpha(in vehicle scale alpha:double=1) -> (Output_Get:double)
   Set in vehicle rigid alpha(execute=@n16.else, in vehicle rigid alpha=1)
   > Set in vehicle scale alpha(in vehicle scale alpha=0)
   Event OnParachuteOpened()
   > Set Is Sky Diving(Is Sky Diving=0)
   > n17 = Switch on EGliderType(Selection=Get PlayerGliderType() -> (PlayerGliderType=HangGlider))
   Set Is Parachute Opened(execute=[@n17.Cape, @n17.HangGlider], Is Parachute Opened:double=1) -> (Output_Get:double)
   Set Is Parachute Opened(execute=[@n17.NoGlider, @n17.Stairs, @n17.Surfing, @n17.SurfingSimple, @n17.Umbrella], Is Parachute Opened=0)
   Event OnParachuteClosed()
   > Set Is Parachute Opened(Is Parachute Opened=0)
  ]}
]
variables [
 clavicleLRot {type=number}
 clavicleRRot {type=number}
 frontendNoEmote {type=number}
 Gravity Override Params {type=GravityOverrideParamsStruct}
 GravityOverride {type=Vector}
 headRotBack {type=number}
 in vehicle rigid alpha {type=number}
 in vehicle scale alpha {type=number}
 Is Parachute Opened {type=number}
 Is Sky Diving {type=number}
 Is Targeting {type=number}
 thigh_up_l {type=number}
 thigh_up_r {type=number}
 ThighSideLf {type=number}
 ThighSideLfIn {type=number}
 ThighSideRt {type=number}
 upperarm_rotY_l {type=number}
 upperarm_rotY_r {type=number}
 upperarm_rotZ_l_Up {type=number}
 upperarm_rotZ_r_Up {type=number}
]
