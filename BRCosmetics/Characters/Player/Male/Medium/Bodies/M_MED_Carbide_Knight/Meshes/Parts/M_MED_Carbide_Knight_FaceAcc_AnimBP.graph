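# Anim Blueprint for the Carbide Knight face accessory that copies the attached parent mesh's pose and adds corrective head rotation.
# Each update derives a backward head-tilt alpha from head versus spine_05, then multiplies it by crouch and skydive flags.
# Two chained additive Modify Bone nodes apply the tilt and crouch rotations.
# A skydive rotation node exists but is not connected to the output pose.
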AnimBlueprint /BRCosmetics/Characters/Player/Male/Medium/Bodies/M_MED_Carbide_Knight/Meshes/Parts/M_MED_Carbide_Knight_FaceAcc_AnimBP
target_skeleton=&/M_MED_Carbide_Knight_FaceAcc_Skeleton
graphs [
 AnimGraph {nodes [
   "Transform (Modify) Bone - Bone: head"(Alpha=Get skydiveHeadRot() -> (skydiveHeadRot:double), Translation="0,0,0", Rotation="0,15,0", Scale="1,1,1") {
    rotationMode=BMM_Additive
    rotationSpace=BCS_BoneSpace
    translationMode=BMM_Additive
    alphaBoolBlend {blendOption=Linear}
    boneToModify {boneName=head}
    rotation {yaw=15}
   }
   n1 = "Transform (Modify) Bone - Bone: head"(
    ComponentPose=Local To Component(LocalPose=Copy Pose From Mesh() {bUseAttachedParent=true})
    Alpha=Get headRotBack() -> (headRotBack:double)
    Translation="0,0,0"
    Rotation="0,10,0"
    Scale="1,1,1"
   ) {
    rotationMode=BMM_Additive
    rotationSpace=BCS_BoneSpace
    translationMode=BMM_Additive
    alphaBoolBlend {blendOption=Linear}
    boneToModify {boneName=head}
    rotation {yaw=10}
   }
   n2 = "Transform (Modify) Bone - Bone: head"(ComponentPose=@n1, Alpha=Get crouchHeadRot() -> (crouchHeadRot:double), Translation="0,0,0", Rotation="0,15,0", Scale="1,1,1") {
    rotationMode=BMM_Additive
    rotationSpace=BCS_BoneSpace
    translationMode=BMM_Additive
    alphaBoolBlend {blendOption=Linear}
    boneToModify {boneName=head}
    rotation {yaw=15}
   }
   Output Pose(Result=Component To Local(ComponentPose=@n2)) {layerGroup=None}
  ]}
 EventGraph {nodes [
   Event BlueprintUpdateAnimation()
   > Set headRotBack(headRotBack:double=GetAlphaFromReferenceTransformDelta(SkeletalMeshComponent=GetOwnerSkeletalMeshComponent(), BoneName=head, BaseName=spine_05, InputRange=(X=0,Y=-50), TwistAxis="1,0,0")) -> (Output_Get:double)
   > Set Is Crouching(Is Crouching:double="To Float (Boolean)"(InBool=Get bIsCrouching())) -> (Output_Get:double)
   n1 = Get headRotBack()
   > Set crouchHeadRot(crouchHeadRot:double=float * float(A=@n1, B=Get Is Crouching() -> (Is Crouching:double)) {is=title K2Node_PromotableOperator}) -> (Output_Get:double)
   > n2 = Set Is Sky Diving(Is Sky Diving:double="To Float (Boolean)"(InBool=Get bIsSkyDiving())) -> (Output_Get:double)
   > Set skydiveHeadRot(skydiveHeadRot:double=float * float(A=@n2, B=@n1)) -> (Output_Get:double)
  ]}
]
variables [
 crouchHeadRot {type=number}
 GetPlayerListUsesCustomCharacterParts {type=boolean}
 headRotBack {type=number}
 Is Crouching {type=number}
 Is Sky Diving {type=number}
 MeshToCopy {type=SkeletalMeshComponent}
 skydiveHeadRot {type=number}
]
